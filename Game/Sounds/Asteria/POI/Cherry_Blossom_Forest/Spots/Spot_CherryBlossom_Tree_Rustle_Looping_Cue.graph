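# Looping Sound Cue that plays intermittent tree rustles as an outdoor ambient spot sound for the Cherry Blossom Forest POI.
# A Looping node repeats a random delay into a quality switch, whose first two levels mix randomized Knights forest rustles, Reality Forest wind-leaves and cherry blossom rustles.
# The last two levels play one smaller cherry blossom pool.
# Modulators vary pitch and volume, with tree-rustle attenuation and overridden concurrency limiting instances.
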SoundCue /Game/Sounds/Asteria/POI/Cherry_Blossom_Forest/Spots/Spot_CherryBlossom_Tree_Rustle_Looping_Cue
root=Looping_0
nodes [
 {id=Delay_0, children [QualityLevel_0], properties {delayMax=12, delayMin=8}}
 {id=Looping_0, children [Delay_0]}
 {id=Mixer_0, children [Modulator_3,Modulator_0,Modulator_1], properties {inputVolume [0.5,0.3,1]}}
 {id=Modulator_0, children [Random_0], properties {pitchMax=1.1, volumeMax=1.1, volumeMin=0.9}}
 {id=Modulator_1, children [Random_1], properties {pitchMax=1.1, pitchMin=1, volumeMax=1.1, volumeMin=1}}
 {id=Modulator_2, children [Random_2], properties {pitchMax=1.1, volumeMax=1.1, volumeMin=1}}
 {id=Modulator_3, children [Random_3], properties {pitchMax=1.1, pitchMin=1, volumeMax=1, volumeMin=0.25}}
 {id=QualityLevel_0, children [Mixer_0,Mixer_0,Modulator_2,Modulator_2]}
 {id=Random_0, children [WavePlayer_0,WavePlayer_1,WavePlayer_3,WavePlayer_5], properties {weights [1,1,1,1]}}
 {
  id=Random_1
  children [WavePlayer_10, WavePlayer_11, WavePlayer_12, WavePlayer_13, WavePlayer_14, WavePlayer_15]
  properties {weights [1,1,1,1,1,1]}
 }
 {id=Random_2, children [WavePlayer_16, WavePlayer_17, WavePlayer_18], properties {weights [1,1,1]}}
 {
  id=Random_3
  children [WavePlayer_2, WavePlayer_4, WavePlayer_6, WavePlayer_8, WavePlayer_19, WavePlayer_20, WavePlayer_21, WavePlayer_22]
  properties {weights [1,1,1,1,1,1,1,1]}
 }
 {
  id=WavePlayer_0
  properties {soundWaveAssetPtr=/Game/Sounds/Artemis/Biomes/Reality_Forest/Spots/AMB_Biome_Reality_Forest_SS_Wind_Leaves_02_06}
 }
 {
  id=WavePlayer_1
  properties {soundWaveAssetPtr=/Game/Sounds/Artemis/Biomes/Reality_Forest/Spots/AMB_Biome_Reality_Forest_SS_Wind_Leaves_02_01}
 }
 {id=WavePlayer_10, properties {soundWaveAssetPtr=&/Spot_CherryBlossom_Tree_Rustle_06}}
 {id=WavePlayer_11, properties {soundWaveAssetPtr=&/Spot_CherryBlossom_Tree_Rustle_01}}
 {id=WavePlayer_12, properties {soundWaveAssetPtr=&/Spot_CherryBlossom_Tree_Rustle_02}}
 {id=WavePlayer_13, properties {soundWaveAssetPtr=&/Spot_CherryBlossom_Tree_Rustle_03}}
 {id=WavePlayer_14, properties {soundWaveAssetPtr=&/Spot_CherryBlossom_Tree_Rustle_04}}
 {id=WavePlayer_15, properties {soundWaveAssetPtr=&/Spot_CherryBlossom_Tree_Rustle_05}}
 {id=WavePlayer_16, properties {soundWaveAssetPtr=&Spot_CherryBlossom_Tree_Rustle_02}}
 {id=WavePlayer_17, properties {soundWaveAssetPtr=&Spot_CherryBlossom_Tree_Rustle_04}}
 {id=WavePlayer_18, properties {soundWaveAssetPtr=&Spot_CherryBlossom_Tree_Rustle_03}}
 {id=WavePlayer_19, properties {soundWaveAssetPtr=/Game/Sounds/Asteria/Biomes/Knights/Spots/Spot_Knights_Forest_Tree_Rustle_17}}
 {id=WavePlayer_2, properties {soundWaveAssetPtr=/Game/Sounds/Asteria/Biomes/Knights/Spots/Spot_Knights_Forest_Tree_Rustle_11}}
 {id=WavePlayer_20, properties {soundWaveAssetPtr=/Game/Sounds/Asteria/Biomes/Knights/Spots/Spot_Knights_Forest_Tree_Rustle_18}}
 {id=WavePlayer_21, properties {soundWaveAssetPtr=/Game/Sounds/Asteria/Biomes/Knights/Spots/Spot_Knights_Forest_Tree_Rustle_19}}
 {id=WavePlayer_22, properties {soundWaveAssetPtr=/Game/Sounds/Asteria/Biomes/Knights/Spots/Spot_Knights_Forest_Tree_Rustle_20}}
 {
  id=WavePlayer_3
  properties {soundWaveAssetPtr=/Game/Sounds/Artemis/Biomes/Reality_Forest/Spots/AMB_Biome_Reality_Forest_SS_Wind_Leaves_02_03}
 }
 {id=WavePlayer_4, properties {soundWaveAssetPtr=/Game/Sounds/Asteria/Biomes/Knights/Spots/Spot_Knights_Forest_Tree_Rustle_12}}
 {
  id=WavePlayer_5
  properties {soundWaveAssetPtr=/Game/Sounds/Artemis/Biomes/Reality_Forest/Spots/AMB_Biome_Reality_Forest_SS_Wind_Leaves_02_05}
 }
 {id=WavePlayer_6, properties {soundWaveAssetPtr=/Game/Sounds/Asteria/Biomes/Knights/Spots/Spot_Knights_Forest_Tree_Rustle_13}}
 {id=WavePlayer_8, properties {soundWaveAssetPtr=/Game/Sounds/Asteria/Biomes/Knights/Spots/Spot_Knights_Forest_Tree_Rustle_15}}
]
properties {
 attenuationSettings=/Game/Sounds/Attenuation/Fort_TileAmbient_OneShot_Tree_Rustle_Smaller_Att
 bEnableBusSends=false
 bOverrideConcurrency=true
 maxDistance=2700
 soundClassObject=&Fort_Ambient_Outdoors_Athena
 volumeMultiplier=0.501
 concurrencyOverrides {maxCount=3, retriggerTime=2, voiceStealReleaseTime=3}
 preEffectBusSends [{}]
}
